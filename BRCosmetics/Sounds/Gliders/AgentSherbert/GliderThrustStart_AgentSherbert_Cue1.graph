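# Sound cue for the AgentSherbert glider's thrust start, pairing a thrust sound with a screech.
# A mixer layers one of three random thrust-start waves with a briefly, randomly delayed screech picked from six equally weighted players, several using a Screech_00 wave.
# A root modulator slightly varies the pitch of the whole mix.
# Shared glider thrust attenuation and concurrency assets govern playback.
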SoundCue /BRCosmetics/Sounds/Gliders/AgentSherbert/GliderThrustStart_AgentSherbert_Cue1
root=Modulator_0
nodes [
 {id=Delay_0, children [Random_1], properties {delayMax=0.25, delayMin=0.1}}
 {id=Mixer_0, children [Random_0,Delay_0], properties {inputVolume [1,1]}}
 {id=Modulator_0, children [Mixer_0], properties {pitchMax=1.1, pitchMin=0.9}}
 {id=Random_0, children [WavePlayer_0,WavePlayer_1,WavePlayer_2], properties {weights [1,1,1]}}
 {
  id=Random_1
  children [WavePlayer_6, WavePlayer_7, WavePlayer_8, WavePlayer_9, WavePlayer_10, WavePlayer_11]
  properties {weights [1,1,1,1,1,1]}
 }
 {id=WavePlayer_0, properties {soundWaveAssetPtr=/Game/Sounds/Fort_Gadgets/Athena/AgentSherbert/Glider_AgentSherbert_ThrustStart_03}}
 {id=WavePlayer_1, properties {soundWaveAssetPtr=/Game/Sounds/Fort_Gadgets/Athena/AgentSherbert/Glider_AgentSherbert_ThrustStart_02}}
 {id=WavePlayer_10, properties {soundWaveAssetPtr=&/Glider_AgentSherbert_Screech_00}}
 {id=WavePlayer_11, properties {soundWaveAssetPtr=&Glider_AgentSherbert_Screech_00}}
 {id=WavePlayer_2, properties {soundWaveAssetPtr=/Game/Sounds/Fort_Gadgets/Athena/AgentSherbert/Glider_AgentSherbert_ThrustStart_01}}
 {id=WavePlayer_6, properties {soundWaveAssetPtr=&/Glider_AgentSherbert_Screech_01}}
 {id=WavePlayer_7, properties {soundWaveAssetPtr=&/Glider_AgentSherbert_Screech_02}}
 {id=WavePlayer_8, properties {soundWaveAssetPtr=&/Glider_AgentSherbert_Screech_03}}
 {id=WavePlayer_9, properties {soundWaveAssetPtr=&Glider_AgentSherbert_Screech_00}}
]
properties {
 attenuationSettings=&Fort_Player_Glider_Thrust_Athena_Att
 maxDistance=12500
 soundClassObject=&Fort_Gliders_Land
 volumeMultiplier=0.5
 concurrencySet [&Athena_Glider_Thrust_Conc]
}
